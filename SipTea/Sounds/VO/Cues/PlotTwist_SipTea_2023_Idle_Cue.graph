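# Sound cue for idle voice-over lines of the PlotTwist SipTea 2023 character, set up as NPC voice.
# A Random node picks among five equally weighted idle wave recordings each time the cue plays.
# Output uses the NPC VO sound class and a shared close-range NPC attenuation asset for spatial falloff.
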SoundCue /SipTea/Sounds/VO/Cues/PlotTwist_SipTea_2023_Idle_Cue
root=Random_1
nodes [
 {id=Random_1, children [WavePlayer_10, WavePlayer_11, WavePlayer_12, WavePlayer_13, WavePlayer_14], properties {weights [1,1,1,1,1]}}
 {id=WavePlayer_10, properties {soundWaveAssetPtr=/SipTea/Sounds/VO/Waves/PlotTwist_SipTea_2023_Idle_010}}
 {id=WavePlayer_11, properties {soundWaveAssetPtr=/SipTea/Sounds/VO/Waves/PlotTwist_SipTea_2023_Idle_020}}
 {id=WavePlayer_12, properties {soundWaveAssetPtr=/SipTea/Sounds/VO/Waves/PlotTwist_SipTea_2023_Idle_040}}
 {id=WavePlayer_13, properties {soundWaveAssetPtr=/SipTea/Sounds/VO/Waves/PlotTwist_SipTea_2023_Idle_050}}
 {id=WavePlayer_14, properties {soundWaveAssetPtr=/SipTea/Sounds/VO/Waves/PlotTwist_SipTea_2023_Idle_051}}
]
properties {
 attenuationSettings=/Game/Sounds/Fort_Human_VO_BR/NPCs/Shared/NPC_VO_Close_Att
 maxDistance=8500
 soundClassObject=&Fort_VO_NPC
 volumeMultiplier=1
 concurrencySet [null]
}
