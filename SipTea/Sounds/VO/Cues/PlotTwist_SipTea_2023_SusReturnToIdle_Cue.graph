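# Sound cue for SipTea NPC voice-over, named for the PlotTwist character's "suspicious return to idle" bark.
# A Random node picks one of five wave variations with equal weights.
# It plays at boosted volume through the NPC VO sound class, with shared close-range NPC attenuation and a SipTea chatter concurrency setting to manage overlapping lines.
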SoundCue /SipTea/Sounds/VO/Cues/PlotTwist_SipTea_2023_SusReturnToIdle_Cue
root=Random_0
nodes [
 {id=Random_0, children [WavePlayer_0,WavePlayer_1,WavePlayer_2,WavePlayer_3,WavePlayer_4], properties {weights [1,1,1,1,1]}}
 {id=WavePlayer_0, properties {soundWaveAssetPtr=/SipTea/Sounds/VO/Waves/PlotTwist_SipTea_2023_SusReturnToIdle_010}}
 {id=WavePlayer_1, properties {soundWaveAssetPtr=/SipTea/Sounds/VO/Waves/PlotTwist_SipTea_2023_SusReturnToIdle_020}}
 {id=WavePlayer_2, properties {soundWaveAssetPtr=/SipTea/Sounds/VO/Waves/PlotTwist_SipTea_2023_SusReturnToIdle_030}}
 {id=WavePlayer_3, properties {soundWaveAssetPtr=/SipTea/Sounds/VO/Waves/PlotTwist_SipTea_2023_SusReturnToIdle_040}}
 {id=WavePlayer_4, properties {soundWaveAssetPtr=/SipTea/Sounds/VO/Waves/PlotTwist_SipTea_2023_SusReturnToIdle_050}}
]
properties {
 attenuationSettings=/Game/Sounds/Fort_Human_VO_BR/NPCs/Shared/NPC_VO_Close_Att
 maxDistance=8500
 soundClassObject=&Fort_VO_NPC
 volumeMultiplier=2
 concurrencySet [/SipTea/Sounds/VO/SipTea_Chatter_Conc]
}
